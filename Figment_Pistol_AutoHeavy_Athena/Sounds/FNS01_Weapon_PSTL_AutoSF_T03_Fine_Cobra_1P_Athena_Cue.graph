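# First-person fire sound cue for the Cobra auto heavy pistol in the Figment Athena content.
# A single close-range SMG single-fire wave plays through a Modulator left at default settings, with a boosted cue volume.
# It uses first-person projectile attenuation, the first-person ranged-fire sound class and a shared first-person gunfire concurrency set.
# The debug flag is on, and attenuation and concurrency override values are also stored.
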SoundCue /Figment_Pistol_AutoHeavy_Athena/Sounds/FNS01_Weapon_PSTL_AutoSF_T03_Fine_Cobra_1P_Athena_Cue
root=Modulator_0
nodes [
 {id=Modulator_0, children [WavePlayer_1]}
 {id=WavePlayer_1, properties {soundWaveAssetPtr=&/sfx_smg_singlefire_close}}
]
properties {
 attenuationSettings=&Fort_Weapon_Projectile_FirstPerson_Athena_Att
 bDebug=true
 maxDistance=2097152
 soundClassObject=&Fort_WeaponRanged_Fire_1P_PSM
 volumeMultiplier=2.5799
 attenuationOverrides {manualReverbSendLevel=-1.0165, reverbSendMethod=Manual}
 concurrencyOverrides {bLimitToOwner=true, maxCount=3, resolutionRule=StopOldest, volumeScale=0.9}
 concurrencySet [&Weapon_Gunfire_FirstPerson_Conc]
}
